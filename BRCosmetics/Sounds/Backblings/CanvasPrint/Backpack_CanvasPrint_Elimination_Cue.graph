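# Sound cue for the CanvasPrint back bling that, judging by its name and reactive property sheet, plays as a cosmetic reaction to eliminations.
# A fixed short delay feeds a single wave player holding the elimination sound, with no randomization or variation.
# Playback runs at reduced volume through the back bling sound class, using close-range pet attenuation and the reactive back bling audio property sheet.
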SoundCue /BRCosmetics/Sounds/Backblings/CanvasPrint/Backpack_CanvasPrint_Elimination_Cue
root=Delay_0
nodes [
 {id=Delay_0, children [WavePlayer_0], properties {delayMax=0.3, delayMin=0.3}}
 {id=WavePlayer_0, properties {soundWaveAssetPtr=&/Backpack_CanvasPrint_Elimination_01}}
]
properties {
 attenuationSettings=&Fort_Pets_Close_Athena_Att
 audioPropertiesSheet=/BRCosmetics/Sounds/Settings/PropertySheets/Reactive/APS_Cosmetic_Reactive_Backbling
 maxDistance=1800.0001
 soundClassObject=/Game/Sounds/FortSoundClassesAndModes/Fort_Backblings
 volumeMultiplier=0.3
}
